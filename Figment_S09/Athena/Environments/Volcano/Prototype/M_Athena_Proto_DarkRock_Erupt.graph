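# Prototype material for dark volcano rock in Athena, layering cliff rock, burnt top, lava ground and grass.
# Height gradients measured from a fixed world position tint the rock and drive the burnt layer, with textures breaking up transitions.
# Vertex colors steer terrain-color blending, burn and lava coverage, while a painted mask places colormap-tinted grass.
# Quality and feature-level switches substitute cheaper rock and burn layers on low-end targets.
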Material /Figment_S09/Athena/Environments/Volcano/Prototype/M_Athena_Proto_DarkRock_Erupt
graph {
 properties {bUseMaterialAttributes=true, pixelDepthOffsetMode=PDOM_Legacy, usedShadingModels=Default Lit}
 nodes [
  Constant() {r=1}
  n1 = GetMaterialAttributes(None=MaterialFunctionCall() {materialFunction=/FigmentEnvironment/LM_8_1_1/Athena/Environments/Landscape/MaterialFunctions/Standard/MF_Athena_Grass_01}.Result) {attributeGetTypes [#1, #2]}
  n2 = Constant3Vector() {constant=<rgba 0 0 1 0>}
  n3 = TextureSample(UVs=MaterialFunctionCall!(WorldPosition) {materialFunction=/FigmentEnvironment/LM_8_1_1/Athena/Environments/Landscape/MaterialFunctions/Utilities/MF_Athena_Terrain_Coords}.Result) {
   samplerSource=SSM_Wrap_WorldGroupSettings
   samplerType=SAMPLERTYPE_LinearColor
   texture=/FigmentEnvironment/LM_8_1_1/Athena/Environments/Landscape/Textures/T_Athena_Terrain_CombinedColors_01
  }
  n4 = Multiply(A=@n1.Base Color, B=@n3.RGB)
  n5 = CollectionParameter() {collection=/FigmentEnvironment/LM_8_1_1/Athena/Environments/Landscape/MPC/MPC_Landscape, parameterName=ColormapDivisor}
  n6 = TextureSampleParameter2D() {
   desc="Painted inside editor (TexturePaint)"
   group=Grass
   parameterName=GrassIDMask
   samplerType=SAMPLERTYPE_Masks
   texture=/Game/Athena/Environments/Volcano/Textures/T_Volcano_M
  }
  n7 = Multiply(A=TextureCoordinate()) {constB=5}
  n8 = TextureSample(UVs=@n7) {samplerSource=SSM_Wrap_WorldGroupSettings, texture=/Game/Packages/DS_Fortnite_Terrain_NoLOD/Textures/T_CliffTileB_D}
  n9 = MaterialFunctionCall(
   Transition Phase=DotProduct(A=@n6.RGB, B=VectorParameter() {group=Grass, parameterName=MaskCh, defaultValue=<rgba 1 0 0 0>}.RGB)
   Height Texture=OneMinus(None=@n8.A)
  ) {materialFunction=&HeightLerp}
  n10 = VertexColor()
  n11 = OneMinus(None=Add(A=@n10.R, B=@n10.R))
  n12 = MaterialFunctionCall!(UV=Multiply(A=TextureCoordinate(), B=ScalarParameter() {defaultValue=30, parameterName=LavaSoil_Tiling}), WorldAlignedSize) {materialFunction=/Game/Environments/Sets/Vulcano/Materials/MF_LavaGround}
  n13 = GetMaterialAttributes(None=@n12.Result) {attributeGetTypes [#2]}
  n14 = VertexColor()
  n15 = Multiply(
   A=@n13.Base Color
   B=LinearInterpolate(
    A=VectorParameter() {parameterName=TIntLavaTransition, defaultValue=<rgba 2 1.3219 0.75 1>}.RGB
    B=Constant3Vector() {constant=<rgba 1 1 1 0>}
    Alpha=@n14.B
   )
  )
  n16 = MaterialFunctionCall!(
   Diffuse 1
   Diffuse 2
   Transistion Phase=OneMinus(None=@n14.B)
   Height Texture 1=@n12.Height
   Height Texture 2=@n8.A
   Contrast=ScalarParameter() {defaultValue=1, parameterName=ContrastLavaBlend}
  ) {materialFunction=/Engine/Functions/Engine_MaterialFunctions02/Texturing/HeightLerpWithTwoHeightMaps}
  n17 = VertexColor()
  n18 = Constant() {r=0.0381}
  n19 = Multiply(A=TextureCoordinate(), B=Constant() {r=18.0373})
  n20 = TextureSample(UVs=@n19) {samplerSource=SSM_Wrap_WorldGroupSettings, texture=/Game/Environments/Sets/Vulcano/Textures/LavaGround_D}
  n21 = ScalarParameter() {defaultValue=0.0125, group=Top, parameterName=BurnDB}
  n22 = ComponentMask(None=Subtract(A=WorldPosition().XYZ, B=NamedRerouteUsage())) {b=true}
  n23 = ScalarParameter() {defaultValue=13952, group=Top, parameterName=BurnHB}
  n24 = Saturate(None=Divide(A=Subtract(A=@n22, B=@n23), B=ScalarParameter() {defaultValue=100, group=Top, parameterName=BurnHBlrB}))
  n25 = Constant() {r=0.15}
  n26 = Constant() {r=2}
  n27 = Multiply(A=TextureCoordinate(), B=Constant() {r=10})
  n28 = TextureSample(UVs=@n27) {samplerSource=SSM_Wrap_WorldGroupSettings, samplerType=SAMPLERTYPE_Normal, texture=/Game/Athena/Environments/Volcano/Textures/T_InkySmoke_N}
  n29 = TextureSample(UVs=@n27) {samplerSource=SSM_Wrap_WorldGroupSettings, texture=/Game/Effects/Fort_Effects/Textures/Tile/T_Inky_Smoke_Tile}
  n30 = MaterialFunctionCall(
   A=MaterialFunctionCall(Normal=@n28.RGB, Flatness=LinearInterpolate(Alpha=@n24) {constA=1, constB=0}) {materialFunction=&FlattenNormal}.Result
   B=Constant3Vector() {constant=<rgba 0 0 1 0>}
   Transition Phase=Constant() {r=0.6331}
   Height Texture=@n29.R
   Contrast=Constant() {r=2.0356}
  ) {materialFunction=&HeightLerp}
  n31 = TextureSample(UVs=@n19) {samplerSource=SSM_Wrap_WorldGroupSettings, samplerType=SAMPLERTYPE_Normal, texture=/Game/Environments/Sets/Vulcano/Textures/LavaGround_N}
  n32 = Constant() {r=0.5}
  n33 = VertexNormalWS()
  n34 = LinearInterpolate(A=Constant() {r=-3.9615}, B=Constant() {r=3.7946}, Alpha=Saturate(None=ComponentMask(None=@n33) {b=true}))
  n35 = SetMaterialAttributes!(
   MaterialAttributes
   Base Color=LinearInterpolate(A=Multiply(A=@n18, B=@n20.R), B=@n21, Alpha=@n24)
   Specular=LinearInterpolate(A=Constant() {r=0.1}, B=@n25, Alpha=@n24)
   Roughness=LinearInterpolate(A=@n26, B=Constant() {r=1}, Alpha=@n24)
   Normal=LinearInterpolate(
    A=@n30.Results
    B=MaterialFunctionCall(Normal=@n31.RGB, Flatness=@n32) {materialFunction=&FlattenNormal}.Result
    Alpha=Saturate(None=@n34)
   )
  ) {attributeSetTypes [
    #2
    #3
    #4
    #1
   ]}
  n36 = SetMaterialAttributes!(MaterialAttributes, Base Color=Add(A=@n21, B=ScalarParameter() {group=Top, parameterName=BurnDB_Mobile}), Specular=@n25, Roughness=@n26) {attributeSetTypes [#2, #3, #4]}
  n37 = QualitySwitch(Default=@n35, Low=@n36, High=@n35, Medium=@n35)
  n38 = VertexColor()
  n39 = ComponentMask(None=Subtract(A=WorldPosition().XYZ, B=NamedRerouteUsage())) {b=true}
  n40 = ScalarParameter() {defaultValue=1007.4706, group=Top, parameterName=BurnH}
  n41 = Saturate(None=Divide(A=Subtract(A=@n39, B=@n40), B=ScalarParameter() {defaultValue=3892.988, group=Top, parameterName=BurnHBlr}))
  n42 = MaterialFunctionCall(Transition Phase=@n41, Height Texture=@n29.R, Contrast=Constant() {r=8.3369}) {materialFunction=&HeightLerp}
  n43 = Saturate(None=Add(A=OneMinus(None=@n38.A), B=@n42.Alpha))
  n44 = ComponentMask(None=Subtract(A=WorldPosition().XYZ, B=NamedRerouteUsage())) {b=true}
  n45 = Constant() {desc=Fixed, r=2480}
  n46 = Saturate(None=Divide(A=Subtract(A=@n44, B=@n45), B=Constant() {r=-4000}))
  n47 = LinearInterpolate(A=Constant() {r=0.68}, Alpha=@n46)
  n48 = Subtract(A=WorldPosition() {worldPositionShaderOffset=WPT_ExcludeAllShaderOffsets}.XYZ, B=NamedRerouteUsage())
  n49 = ComponentMask(None=@n48) {b=true}
  n50 = Divide(
   A=Subtract(A=@n49, B=ScalarParameter() {defaultValue=-170.6412, group=Colors, parameterName=GradLowPos})
   B=ScalarParameter() {defaultValue=-1849.3988, group=Colors, parameterName=GradLowPosBlur}
  )
  n51 = Subtract(A=@n49, B=ScalarParameter() {defaultValue=-689.9945, group=Colors, parameterName=GradMidPos})
  n52 = ScalarParameter() {defaultValue=1366.7939, group=Colors, parameterName=GradLowPosBlure}
  n53 = LinearInterpolate(
   A=LinearInterpolate(
    A=VectorParameter() {group=Colors, parameterName=GradientMid, defaultValue=<rgba 0.0469 0.0303 0.0195 0>}.RGB
    B=VectorParameter() {group=Colors, parameterName=GradientBottom, defaultValue=<rgba 0.0394 0.0417 0.0278 0>}.RGB
    Alpha=Saturate(None=@n50)
   )
   B=VectorParameter() {group=Colors, parameterName=GradientTop, defaultValue=<rgba 0.0365 0.0365 0.0365 0>}.RGB
   Alpha=Saturate(None=Divide(A=@n51, B=@n52))
  )
  n54 = TextureSample(UVs=Multiply(A=TextureCoordinate(), B=Constant() {r=2.8233})) {samplerSource=SSM_Wrap_WorldGroupSettings, samplerType=SAMPLERTYPE_Normal, texture=/Game/Athena/Environments/Volcano/Textures/Detail_normal}
  n55 = TextureSample(UVs=@n7) {
   samplerSource=SSM_Wrap_WorldGroupSettings
   samplerType=SAMPLERTYPE_Normal
   texture=/Game/Packages/DS_Fortnite_Terrain_NoLOD/Textures/T_CliffTileB_N
  }
  n56 = MaterialFunctionCall(A=@n54.RGB, B=@n55.RGB, Transition Phase=Constant() {r=0.5476}, Height Texture=@n8.A, Contrast=Constant() {r=5}) {materialFunction=&HeightLerp}
  n57 = Constant3Vector() {constant=<rgba 5 3.0657 1.1979 1>}
  n58 = Constant3Vector() {constant=<rgba 2 1.2263 0.4792 1>}
  n59 = VertexColor()
  n60 = MaterialFunctionCall(
   A=Constant() {r=1}
   B=LinearInterpolate(B=LinearInterpolate(A=@n57, B=@n58, Alpha=@n59.G), Alpha=@n46) {constA=1}
   Transition Phase=Constant() {r=0.57}
   Height Texture=@n8.R
   Contrast=Constant() {r=0.7484}
  ) {materialFunction=&HeightLerp}
  n61 = Multiply(A=LinearInterpolate(A=Multiply(A=@n47, B=@n53), B=@n53, Alpha=@n56.Lerp Alpha No Contrast) {constA=1}, B=@n60.Results)
  n62 = ScalarParameter() {defaultValue=0.75, group=Colors, parameterName=TopDarkGradient}
  n63 = Constant() {r=0.15}
  n64 = Constant() {r=1.5}
  n65 = MakeMaterialAttributes!(
   BaseColor=LinearInterpolate(A=Multiply(A=@n61, B=@n62), B=@n61, Alpha=@n46)
   Metallic=VertexColor().G
   Specular=LinearInterpolate(A=Constant() {r=0.05}, B=LinearInterpolate(A=@n63, B=@n64, Alpha=@n60.Alpha), Alpha=@n46)
   Roughness=Constant() {r=1.25}
   Anisotropy
   EmissiveColor
   Opacity
   OpacityMask
   Normal=LinearInterpolate(A=@n54.RGB, B=@n56.Results, Alpha=@n46)
   Tangent
   WorldPositionOffset
   SubsurfaceColor
   ClearCoat
   ClearCoatRoughness
   AmbientOcclusion
   Refraction
   CustomizedUVs_0
   CustomizedUVs_1
   CustomizedUVs_2
   CustomizedUVs_3
   CustomizedUVs_4
   CustomizedUVs_5
   CustomizedUVs_6
   CustomizedUVs_7
   PixelDepthOffset
   ShadingModel
   Displacement
  )
  n66 = Constant() {r=0.25}
  n67 = Constant() {r=1}
  n68 = SetMaterialAttributes!(MaterialAttributes, Base Color=@n53, Specular=@n66, Roughness=@n67, Normal=@n55.RGB) {attributeSetTypes [
    #2
    #3
    #4
    #1
   ]}
  n69 = QualitySwitch(Default=@n65, Low=@n68, High=@n65, Medium=@n65)
  n70 = SetMaterialAttributes(Base Color=@n53, Specular=@n66, Roughness=@n67, Normal=Constant3Vector() {constant=<rgba 0 0 1 0>}) {attributeSetTypes [
    #2
    #3
    #4
    #1
   ]}
  n71 = GetMaterialAttributes(None=FeatureLevelSwitch(Default=@n69, ES2=@n70, ES3_1=@n68, SM4_REMOVED=@n69, SM5=@n69, SM6=@n69)) {attributeGetTypes [#2, #3, #4]}
  n72 = VectorParameter() {parameterName=BlendColortoTerrain, defaultValue=<rgba 0.4688 0.3601 0.1797 1>}
  n73 = VectorParameter() {parameterName=BlendColor_Low, defaultValue=<rgba 0.3906 0.3384 0.2055 1>}
  n74 = Saturate(None=Multiply(A=VertexColor().R, B=NamedRerouteUsage()))
  n75 = LinearInterpolate(A=@n71.Base Color, B=Multiply(A=NamedRerouteUsage(), B=QualitySwitch(Default=@n72.RGB, Low=@n73.RGB, Medium=@n73.RGB)), Alpha=@n74)
  n76 = NamedRerouteUsage()
  n77 = LinearInterpolate(
   A=@n71.Specular
   B=QualitySwitch(Default=ScalarParameter() {defaultValue=0.4, parameterName=landscape Spec}, Low=Multiply(A=@n76) {constB=0.3})
   Alpha=@n74
  )
  n78 = ScalarParameter() {defaultValue=0.9, parameterName=landscape roughness}
  n79 = Constant() {r=1}
  n80 = LinearInterpolate(A=@n71.Roughness, B=QualitySwitch(Default=@n78, Low=@n79, High=@n78, Medium=@n79, Epic=@n78), Alpha=@n74)
  n81 = GetMaterialAttributes(None=MaterialFunctionCall!(
    Base Material=SetMaterialAttributes!(MaterialAttributes=@n71.MaterialAttributes, Base Color=@n75, Specular=@n77, Roughness=@n80) {attributeSetTypes [#2, #3, #4]}
    Top Material=FeatureLevelSwitch(Default=@n37, ES2=@n36, ES3_1=@n36, SM4_REMOVED=@n37, SM5=@n37, SM6=@n37)
    Alpha=FeatureLevelSwitch(Default=@n43, ES2=@n41, ES3_1=@n41, SM4_REMOVED=@n43, SM5=@n43, SM6=@n43)
   ) {materialFunction=/Engine/Functions/MaterialLayerFunctions/MatLayerBlend_Standard}.Blended Material) {attributeGetTypes [#1]}
  n82 = TextureSample(UVs=Multiply(A=Constant() {r=35}, B=TextureCoordinate())) {samplerType=SAMPLERTYPE_Normal, texture=/Game/Athena/Environments/Volcano/Textures/Small_Detail_normal}
  n83 = Constant3Vector() {constant=<rgba 6 6 0 1>}
  n84 = Constant3Vector() {constant=<rgba 1.5 1.5 0 1>}
  n85 = LinearInterpolate(
   A=LinearInterpolate(A=Constant3Vector() {constant=<rgba 0.5 0.5 0 1>}, B=LinearInterpolate(A=@n83, B=@n84, Alpha=@n8.A), Alpha=@n46)
   B=LinearInterpolate(A=Constant3Vector() {constant=<rgba 2 2 0 1>}, B=Constant3Vector() {constant=<rgba 1 1 0 1>}, Alpha=@n30.Alpha)
   Alpha=@n42.Alpha
  )
  n86 = VertexColor()
  n87 = Add(A=@n81.Normal, B=LinearInterpolate(A=Multiply(A=@n82.RGB, B=@n85), Alpha=Max(A=@n86.R, B=@n86.R)) {constB=0})
  MaterialFunctionCall(
   Base Material=MaterialFunctionCall(
    Base Material=SetMaterialAttributes!(MaterialAttributes=@n81.MaterialAttributes, Normal=@n87) {attributeSetTypes [#1]}
    Top Material=SetMaterialAttributes!(MaterialAttributes=@n13.MaterialAttributes, Base Color=@n15) {attributeSetTypes [#2]}
    Alpha=Multiply(A=@n16.Alpha, B=@n17.A)
   ) {materialFunction=&MatLayerBlend_Standard}.Blended Material
   Top Material=SetMaterialAttributes!(MaterialAttributes=@n1.MaterialAttributes, Normal=LinearInterpolate(A=@n1.Normal, B=@n2) {constAlpha=0.6}, Base Color=Multiply(A=@n4, B=@n5)) {attributeSetTypes [#1, #2]}
   Alpha=Saturate(None=Multiply(A=@n9.Lerp Alpha No Contrast, B=@n11))
  ) {materialFunction=&MatLayerBlend_Standard}
  NamedRerouteDeclaration(None=@n8.R) {name=Cliff texture, nodeColor=<rgba 0 1 0.6953 1>}
  NamedRerouteDeclaration(None=Constant3Vector() {constant=<rgba 84480 36864 11520 0>}) {name=Position, nodeColor=<rgba 1 0.9141 0 1>}
  n88 = Divide(
   A=Subtract(A=ComponentMask(None=@n48) {r=true}, B=ScalarParameter() {defaultValue=-13586.6436, group=Colors, parameterName=Transition_Pos})
   B=ScalarParameter() {defaultValue=2359.332, group=Colors, parameterName=Transition_POS_Blur}
  )
  NamedRerouteDeclaration(None=Saturate(None=OneMinus(None=@n88))) {name=wpo_mask, nodeColor=<rgba 1 0.8438 0 1>}
  SetMaterialAttributes!(MaterialAttributes, Base Color, Roughness=Constant() {r=0.9}, Specular=Constant() {r=0.1}) {attributeSetTypes [#2, #4, #3]}
  Transform(None=@n1.Normal)
 ]
}
